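# Masked lit material for the oil liquid effect on the Fowl skin of the SquibblySquid backpack, built for Niagara mesh particles and ribbons.
# Panning noise and UV gradients shape vertex offset, combining a local-space pull with sine-shaped normal displacement scaled by a dynamic parameter.
# Flattened swirl normals, cubemap reflection and a tintable chromatic noise layer supply surface detail and emissive.
# Stepped masks cut opacity, with particle alpha wiping along U.
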Material /BRCosmetics/Accessories/FORT_Backpacks/Backpack_M_MED_SquibblySquid/Skins/Fowl/FX/M_SquibblySquid_Oil_Liquid
graph {
 properties {
  bUsedWithNiagaraMeshParticles=true
  bUsedWithNiagaraRibbons=true
  blendMode=BLEND_Masked
  pixelDepthOffsetMode=PDOM_Legacy
  previewMesh=/Engine/Transient.MaterialEditorMeshComponent_1
  translucencyLightingMode=TLM_Surface
  usedShadingModels=Default Lit
 }
 nodes [
  NamedRerouteDeclaration(None=TextureSample(UVs=NamedRerouteUsage()) {samplerType=SAMPLERTYPE_Grayscale, texture=&LowResBlurredNoise}.RGB) {name=Noise_Sample, nodeColor=<rgba 0.1484 1 0 1>}
  n1 = TextureCoordinate()
  n2 = VectorParameter() {parameterName=UV_Noise, channelNames {a=V_Speed, b=U_Speed, g=V_Scale, r=U_Scale}, defaultValue=<rgba 2 1.8 0 0.65>}
  n3 = AppendVector(A=@n2.U_Scale, B=@n2.U_Scale)
  NamedRerouteDeclaration(None=Panner(Coordinate=Multiply(A=@n1, B=@n3), Speed=AppendVector(A=@n2.U_Speed, B=@n2.U_Speed)) {speedY=0.65}) {name=Noise_UV, nodeColor=<rgba 1 0 0.4687 1>}
  n4 = NamedRerouteUsage()
  NamedRerouteDeclaration(None=MaterialFunctionCall(
    Normal=TextureSample(UVs=@n4) {samplerType=SAMPLERTYPE_Normal, texture=/Game/Global/Textures/Default/Noise/T_SwirlNoise_N}.RGB
    Flatness=Constant()
   ) {materialFunction=&FlattenNormal}.Result) {name=Noise_Normal, nodeColor=<rgba 0 0.7891 1 1>}
  n5 = TextureCoordinate()
  NamedRerouteDeclaration(None=OneMinus(None=ComponentMask(None=@n5) {g=true})) {name=LinearGradient, nodeColor=<rgba 1 0.8906 0 1>}
  n6 = Transform(None=Constant3Vector() {constant=<rgba 2.5 5 0 0>}) {transformSourceType=TRANSFORMSOURCE_Local}
  n7 = Saturate(None=Power(Base=NamedRerouteUsage()) {constExponent=10})
  n8 = TextureCoordinate()
  n9 = Sine(None=Multiply(A=MaterialFunctionCall(Float2=@n8) {materialFunction=&BreakOutFloat2Components}.G) {constB=0.5})
  n10 = Multiply(A=VertexNormalWS(), B=Multiply(A=Saturate(None=@n9), B=NamedRerouteUsage()))
  n11 = Multiply(A=ScalarParameter() {defaultValue=40, parameterName=WPO}, B=DynamicParameter() {paramNames [WPO,Param2,Param3,Param4]}.WPO)
  NamedRerouteDeclaration(None=Add(A=LinearInterpolate(B=@n6, Alpha=@n7), B=Multiply(A=@n10, B=@n11))) {name=WPO, nodeColor=<rgba 0.5703 0 1 1>}
  n12 = NamedRerouteUsage()
  n13 = TextureSampleParameterCube!(UVs=ReflectionVectorWS(CustomWorldNormal=Transform(None=@n12)), Apply View MipBias) {parameterName=ReflectHDRI, samplerType=SAMPLERTYPE_LinearColor, texture=/Game/Effects/FlipSimulation/Rendering/T_WaterReflection}
  NamedRerouteDeclaration(None=Multiply(A=@n13.G, B=ScalarParameter() {defaultValue=0.1, parameterName=Reflect_Intensity}) {constB=0.1}) {name=Specular, nodeColor=<rgba 0 1 0.25 1>}
  n14 = NamedRerouteUsage()
  n15 = Add(
   A=TextureSample(UVs=NamedRerouteUsage()) {samplerType=SAMPLERTYPE_Grayscale, texture=&LowResBlurredNoise}.R
   B=Multiply(A=TextureCoordinate()) {constB=2}
  ) {constB=0.5}
  n16 = Divide(A=Multiply(
    A=TextureSample(UVs=@n15) {texture=&T_NoisePack_01}.RGB
    B=VectorParameter() {parameterName=ChromaticColorScale, defaultValue=<rgba 0 0 0 1>}.RGB
   )) {constB=50}
  NamedRerouteDeclaration(None=MaterialFunctionCall(Emissive=Add(A=@n14, B=@n16)) {materialFunction=&MF_FX_EmissiveAdjustments}.Result) {name=Emissive, nodeColor=<rgba 1 0 0.0703 1>}
  n17 = Saturate(None=Power(Base=NamedRerouteUsage(), Exp=ScalarParameter() {defaultValue=0.5, parameterName=OpacityMaskPower}))
  n18 = ComponentMask(None=NamedRerouteUsage()) {g=true}
  n19 = OneMinus(None=ComponentMask(None=TextureCoordinate()) {r=true})
  n20 = ParticleColor()
  NamedRerouteDeclaration(None=Multiply(A=Step(Y=@n17, X=@n18), B=Step(Y=@n19, X=@n20.A))) {name=Opacity, nodeColor=<rgba 0 0.3906 1 1>}
  NamedRerouteUsage()
  NamedRerouteUsage()
  NamedRerouteUsage()
  NamedRerouteUsage()
 ]
}
